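# Sound cue for a Slone payphone voice line, SLONE_2021_QUEST_21_030, apparently quest dialogue.
# A single wave player feeds the matching wave straight to the output, with no randomization or modulation.
# The cue uses the shared S17 payphone VO attenuation asset and the narrator VO sound class.
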SoundCue /NPCLibrary/Audio/Slone/Payphone/Cues/SLONE_2021_QUEST_21_030_Cue
root=WavePlayer_0
nodes [{id=WavePlayer_0, properties {soundWaveAssetPtr=/NPCLibrary/Audio/Slone/Payphone/Waves/SLONE_2021_QUEST_21_030}}]
properties {
 attenuationSettings=/NPCLibrary/Audio/Slone/Payphone/S17_Payphone_VO_Att
 maxDistance=5400
 soundClassObject=/Game/Sounds/FortSoundClassesAndModes/Fort_VO_Narrator
 volumeMultiplier=1
}
