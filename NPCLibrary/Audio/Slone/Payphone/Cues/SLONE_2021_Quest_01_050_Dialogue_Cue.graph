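# Sound cue that plays a Slone payphone quest dialogue line (Quest 01, line 050) from the NPC library.
# Its single Dialogue Player node plays the matching Dialogue Wave in a context with the S17PayphoneVO speaker and no targets.
# The cue routes through the narrator VO sound class and uses a dedicated payphone VO attenuation asset to spatialize it within a limited range.
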SoundCue /NPCLibrary/Audio/Slone/Payphone/Cues/SLONE_2021_Quest_01_050_Dialogue_Cue
root=DialoguePlayer_1
nodes [{
  id=DialoguePlayer_1
  properties {dialogueWaveParameter {
    dialogueWave=/NPCLibrary/Audio/Slone/Payphone/Waves/SLONE_2021_Quest_01_050_Dialogue
    context {speaker=/NPCLibrary/Audio/Slone/Payphone/S17PayphoneVO, targets []}
   }}
 }]
properties {
 attenuationSettings=/NPCLibrary/Audio/Slone/Payphone/S17_Payphone_VO_Att
 maxDistance=5400
 soundClassObject=/Game/Sounds/FortSoundClassesAndModes/Fort_VO_Narrator
 volumeMultiplier=1
}
